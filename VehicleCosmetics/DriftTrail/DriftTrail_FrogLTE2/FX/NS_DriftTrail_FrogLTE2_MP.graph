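# Green FrogLTE2 drift trail vehicle cosmetic, built from ribbons, wheel-wrap meshes and short-lived cones.
# Two ribbon emitters each trail a pair of ribbons from skeletal positions, shifted by User.offset.
# Cone and spinning wrap emitters attach to wheel bones or sockets by partition index, using FN_DisableTireFX for per-wheel masking.
# User.bIsFrontendPreview switches between two cone emitters; painted-variant overrides are disabled in every emitter, leaving the colors hard-coded.
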NiagaraSystem /VehicleCosmetics/DriftTrail/DriftTrail_FrogLTE2/FX/NS_DriftTrail_FrogLTE2_MP
dependencies {
 usedDataInterfaces [&NiagaraDataInterfaceCurve, &NiagaraDataInterfaceVectorCurve, &NiagaraDataInterfaceSkeletalMesh]
 usedDynamicInputs []
 usedModules [
  &self:SystemSpawnScript.SystemScriptSource.SystemScriptGraph.NiagaraNodeAssignment_0.SetVariables_#1
  &FN_UnpackVehicleFlags
  /VehicleCosmetics/FX/Niagara/Modules/VC_SetVehicleSpeedAttribs
  &SystemState
 ]
}
emitters {
 Garage_Cone {
  emitterUpdateScript [
   &EmitterState(Inactive Response=::NewEnumerator1, Life Cycle Mode=::NewEnumerator1, Loop Duration=0.39, UseLoopDelay=false)
   &SpawnRate(
    SpawnRate=&MakeCustomFloatFromBool(Bool=$User.bIsFrontendPreview, True Float=30, False Float=0)
    Use Spawn Probability=false
   )
  ]
  particleSpawnScript [
   &VC_PaintedVariantsOverride() {enabled=false}
   &InitializeParticle(
    Color Mode=::NewEnumerator0
    Lifetime=0.3
    Mesh Scale Max=<1.955 1.955 1.955>
    Mesh Scale Min=<1.955 1 1.955>
    Mesh Scale Mode=::NewEnumerator2
    Sprite Rotation Mode=::NewEnumerator2
    UsePositionOffset=false
    Write Material Random=true
   )
   &PartitionParticles()
   &SkeletalMeshLocation(
    Bone / Socket Index=&Add_Integer(A=$Particles.PartitionParticles.Partition, B=4)
    Bone / Socket Sampling Mode=::NewEnumerator5
    Interpolate Mesh Sampling=false
    MaskSpawnGroup=false
    Mesh Sampling Type=::NewEnumerator6
    Sampled Position Offset=&AddVectorToPosition(Position=<-37 0 -35>, Vector=$User.offset)
    Skeletal Mesh=$System.SkeletalMesh
   )
   &AddVelocity(Velocity=&VC_Mask_Vectors(
     FrontEnd Value:Vector3f=<-1500 0 0>
     blsFrontEndPreview:NiagaraBool=false
     SpeedMin Value:Vector3f=<-800 0 0>
     SpeedMax Value:Vector3f=<-1 0 0>
     Speed Alpha:NiagaraFloat=$System.Speed_Alpha
    )) {enabled=false}
   &Color(Color=<rgba 0.248 0.8 0.064 1>) {module=Color001}
   &InitialMeshOrientation(
    Mesh Orientation Mode=::NewEnumerator2
    Orientation Coordinate Space=::Local
    Rotation=&RandomRangeVector(Minimum=<0 0.02 0>, Maximum=<0 0.05 0>)
    Use Rotation Vector=true
   ) {module=InitialMeshOrientation001}
  ]
  particleUpdateScript [
   &FN_DisableTireFX(
    Mask Index Attribute:NiagaraInt32=$Particles.PartitionParticles.Partition
    Mask Index FL=2
    Mask Index FR=3
    Mask Index RL=0
    Mask Index RR=1
   )
   &ParticleState()
   &UpdateMeshOrientation(Rotation Rate=&SelectFloatFromArray(
     Array Sampling Mode=::NewEnumerator0
     Direct Array Index=?
     Float Selection Array=DataInterface() {propertyValues {FloatData [0.5,-0.3]}}
     Random Seed=0
    ))
   &ScaleMeshSize(Scale Factor=&RandomRangeVector(
     Minimum=&VectorFromCurve(
      VectorCurve=DataInterface() {propertyValues {
        XCurve {keys [
          <at 0 0 RCIM_Cubic>
          {arriveTangent=0.7778, interpMode=RCIM_Cubic, leaveTangent=0.7778, time=0.1, value=1}
          {arriveTangent=-1.2346, interpMode=RCIM_Cubic, leaveTangent=-1.2346, time=0.9, value=0.7}
          <at 0.91 0 RCIM_Constant>
         ]}
        YCurve {keys [
          <at 0 0 RCIM_Cubic>
          {arriveTangent=0.7778, interpMode=RCIM_Cubic, leaveTangent=0.7778, time=0.1, value=1}
          {arriveTangent=-1.2346, interpMode=RCIM_Cubic, leaveTangent=-1.2346, time=0.9, value=0.7}
          <at 0.91 0 RCIM_Constant>
         ]}
        ZCurve {keys [<at 0 0>,<at 0.9 1>,<at 0.91 0 RCIM_Constant>]}
       }}
      Scale Curve=<1 1 2.5>
     )
     Maximum=&VectorFromCurve(
      VectorCurve=DataInterface() {propertyValues {
        XCurve {keys [
          <at 0 0 RCIM_Cubic>
          {arriveTangent=0.7778, interpMode=RCIM_Cubic, leaveTangent=0.7778, time=0.1, value=1.5}
          {arriveTangent=-1.8519, interpMode=RCIM_Cubic, leaveTangent=-1.8519, time=0.9, value=0.7}
          <at 0.91 0 RCIM_Constant>
         ]}
        YCurve {keys [
          <at 0 0 RCIM_Cubic>
          {arriveTangent=0.7778, interpMode=RCIM_Cubic, leaveTangent=0.7778, time=0.1, value=1.5}
          {arriveTangent=-1.8519, interpMode=RCIM_Cubic, leaveTangent=-1.8519, time=0.9, value=0.7}
          <at 0.91 0 RCIM_Constant>
         ]}
        ZCurve {keys [<at 0 0>,<at 0.9 1>,<at 0.91 0 RCIM_Constant>]}
       }}
      Scale Curve=<1.35 1 1.5>
     )
    )) {module=ScaleMeshSize001}
   &ScaleColor(
    Scale RGB=&VectorFromFloat(Value=&RandomRangeFloat(Minimum=1, Maximum=4))
    ScaleA=true
    ScaleRGB=true
   )
   &DynamicMaterialParameters(
    Index 0 Param 1=&FloatFromCurve(FloatCurve=DataInterface() {propertyValues {Curve {keys [{tangentMode=RCTM_User, time=0, value=1.5}, {tangentMode=RCTM_User, time=1, value=1.5}]}}})
    Index 0 Param 2=&FloatFromCurve(
     FloatCurve=DataInterface() {propertyValues {Curve {keys [<at 0 -1 RCIM_Cubic>,<at 1 0.8 RCIM_Cubic>]}}}
     Scale Curve=&RandomRangeFloat(Minimum=0.5, Maximum=1)
    )
    Index 0 Param 3=1.5
    Module.Index0Float0Write=true
    Module.Index0Float1Write=true
    Module.Index0Float2Write=true
    Module.Index0Float3Write=false
   )
   &SolveForcesAndVelocity(Clamp Velocity=false, Limit Acceleration=false)
  ]
  rendererClasses [&NiagaraMeshRendererProperties, &NiagaraSpriteRendererProperties]
  renderers [
   {rendererClass=&NiagaraMeshRendererProperties, rendererIndex=0}
   {rendererClass=&NiagaraSpriteRendererProperties, rendererIndex=1}
  ]
 }
 Garage_Cone001 {
  emitterUpdateScript [
   &EmitterState(Inactive Response=::NewEnumerator1, Life Cycle Mode=::NewEnumerator1, Loop Duration=0.39, UseLoopDelay=false)
   &SpawnRate(
    SpawnRate=&MakeCustomFloatFromBool(Bool=$User.bIsFrontendPreview, True Float=0, False Float=30)
    Use Spawn Probability=false
   )
  ]
  particleSpawnScript [
   &VC_PaintedVariantsOverride() {enabled=false}
   &InitializeParticle(
    Color Mode=::NewEnumerator0
    Lifetime=0.3
    Mesh Scale Mode=::NewEnumerator3
    Mesh Uniform Scale=1.955
    Sprite Rotation Mode=::NewEnumerator2
    UsePositionOffset=false
    Write Material Random=true
   )
   &PartitionParticles()
   &SkeletalMeshLocation(
    Bone / Socket Index=&Add_Integer(A=$Particles.PartitionParticles.Partition, B=4)
    Bone / Socket Sampling Mode=::NewEnumerator5
    Interpolate Mesh Sampling=false
    MaskSpawnGroup=false
    Mesh Sampling Type=::NewEnumerator6
    Sampled Position Offset=&AddVectorToPosition(Position=<-37 0 -35>, Vector=$User.offset)
    Skeletal Mesh=$System.SkeletalMesh
   )
   &AddVelocity(Velocity=&VC_Mask_Vectors(
     FrontEnd Value:Vector3f=<-1500 0 0>
     blsFrontEndPreview:NiagaraBool=false
     SpeedMin Value:Vector3f=<-800 0 0>
     SpeedMax Value:Vector3f=<-1 0 0>
     Speed Alpha:NiagaraFloat=$System.Speed_Alpha
    )) {enabled=false}
   &Color(Color=<rgba 0.248 0.8 0.064 1>) {module=Color001}
   &InitialMeshOrientation(
    Mesh Orientation Mode=::NewEnumerator2
    Orientation Coordinate Space=::Local
    Rotation=&RandomRangeVector(Minimum=<0 0.02 0>, Maximum=<0 0.05 0>)
    Use Rotation Vector=true
   ) {module=InitialMeshOrientation001}
  ]
  particleUpdateScript [
   &FN_DisableTireFX(
    Mask Index Attribute:NiagaraInt32=$Particles.PartitionParticles.Partition
    Mask Index FL=2
    Mask Index FR=3
    Mask Index RL=0
    Mask Index RR=1
   )
   &ScaleMeshSize(Scale Factor=&VectorFromCurve(VectorCurve=DataInterface() {propertyValues {
       XCurve {keys [<at 0 0.04>,<at 0.45 0.65>,<at 1 2>]}
       YCurve {keys [<at 0 0.38>,<at 0.45 0.86>,<at 1 1.2>]}
       ZCurve {keys [<at 0 0.38>,<at 0.45 0.86>,<at 1 1.2>]}
      }})) {enabled=false}
   &ParticleState()
   &ScaleMeshSize(Scale Factor=&RandomRangeVector(
     Minimum=&VectorFromCurve(
      VectorCurve=DataInterface() {propertyValues {
        XCurve {keys [
          <at 0 0 RCIM_Cubic>
          {arriveTangent=0.7778, interpMode=RCIM_Cubic, leaveTangent=0.7778, time=0.1, value=1}
          {arriveTangent=-1.2346, interpMode=RCIM_Cubic, leaveTangent=-1.2346, time=0.9, value=0.7}
          <at 0.91 0 RCIM_Constant>
         ]}
        YCurve {keys [
          <at 0 0 RCIM_Cubic>
          {arriveTangent=0.7778, interpMode=RCIM_Cubic, leaveTangent=0.7778, time=0.1, value=1}
          {arriveTangent=-1.2346, interpMode=RCIM_Cubic, leaveTangent=-1.2346, time=0.9, value=0.7}
          <at 0.91 0 RCIM_Constant>
         ]}
        ZCurve {keys [<at 0 0.5>,<at 0.9 1>,<at 0.91 0 RCIM_Constant>]}
       }}
      Scale Curve=<1 1 1.5>
     )
     Maximum=&VectorFromCurve(
      VectorCurve=DataInterface() {propertyValues {
        XCurve {keys [
          <at 0 0 RCIM_Cubic>
          {arriveTangent=0.7778, interpMode=RCIM_Cubic, leaveTangent=0.7778, time=0.1, value=1.5}
          {arriveTangent=-1.8519, interpMode=RCIM_Cubic, leaveTangent=-1.8519, time=0.9, value=0.7}
          <at 0.91 0 RCIM_Constant>
         ]}
        YCurve {keys [
          <at 0 0 RCIM_Cubic>
          {arriveTangent=0.7778, interpMode=RCIM_Cubic, leaveTangent=0.7778, time=0.1, value=1.5}
          {arriveTangent=-1.8519, interpMode=RCIM_Cubic, leaveTangent=-1.8519, time=0.9, value=0.7}
          <at 0.91 0 RCIM_Constant>
         ]}
        ZCurve {keys [<at 0 0.5>,<at 0.9 1>,<at 0.91 0 RCIM_Constant>]}
       }}
      Scale Curve=<1.35 1 1.5>
     )
    )) {module=ScaleMeshSize001}
   &ScaleColor(
    Scale RGB=&VectorFromFloat(Value=&RandomRangeFloat(Minimum=1, Maximum=4))
    ScaleA=true
    ScaleRGB=true
   )
   &DynamicMaterialParameters(
    Index 0 Param 1=&FloatFromCurve(FloatCurve=DataInterface() {propertyValues {Curve {keys [{tangentMode=RCTM_User, time=0, value=1.5}, {tangentMode=RCTM_User, time=1, value=1.5}]}}})
    Index 0 Param 2=&FloatFromCurve(FloatCurve=DataInterface() {propertyValues {Curve {keys [<at 0 -1>,<at 0.9 1>]}}})
    Index 0 Param 3=1.5
    Module.Index0Float0Write=true
    Module.Index0Float1Write=true
    Module.Index0Float2Write=true
    Module.Index0Float3Write=false
   )
   &SolveForcesAndVelocity(Clamp Velocity=false, Limit Acceleration=false)
  ]
  rendererClasses [&NiagaraMeshRendererProperties, &NiagaraSpriteRendererProperties]
  renderers [
   {rendererClass=&NiagaraMeshRendererProperties, rendererIndex=0}
   {rendererClass=&NiagaraSpriteRendererProperties, rendererIndex=1}
  ]
 }
 MainTrail {
  emitterUpdateScript [
   &EmitterState()
   &VC_SpawnRibbon_v1(Number of Ribbons=2, Spawn Time=-1, SpawnRate=120)
  ]
  particleSpawnScript [
   &VC_PaintedVariantsOverride() {enabled=false}
   &VC_InitializeRibbon_v1(
    Custom Offset=true
    Enable Skeletal Position=true
    Lifetime=0.3
    Local Vector Offset=&AddVectorToPosition(Position=<-14 0 -32>, Vector=$User.offset)
    Offset Amount=1
    Ribbon Width=57.5
   )
   &AddVelocity(Velocity=<-1400 0 0>, Velocity Speed Scale=2) {module=AddVelocity001}
   &Color(Color=<rgba 0.248 0.8 0.064 1>)
   &Color(Color=<rgba 0.4 0.8963 0.0296 1>) {enabled=false, module=Color001}
  ]
  particleUpdateScript [
   &ParticleState()
   &SolveForcesAndVelocity(Clamp Velocity=false, Limit Acceleration=false)
   &ScaleColor(Scale Alpha=5, Scale RGB=&VectorFromFloat(Value=6), ScaleA=true, ScaleRGB=true)
   &ScaleRibbonWidth(Ribbon Width Scale=&FloatFromCurve(FloatCurve=DataInterface() {propertyValues {Curve {keys [
         <at 0 0 RCIM_Cubic>
         {arriveTangent=1.7626, interpMode=RCIM_Cubic, leaveTangent=1.7626, tangentMode=RCTM_User, time=0.1, value=0.75}
         <at 0.5 1 RCIM_Cubic>
        ]}}}))
   &VC_UpdateRibbon_v1(Enable Skeletal Position=true)
   &DynamicMaterialParameters(
    Index 0 Param 1=1
    Index 0 Param 2=0.5
    Index 0 Param 3=0.01
    Index 0 Param 4=1
    Module.Index0Float0Write=true
    Module.Index0Float1Write=true
    Module.Index0Float2Write=true
    Module.Index0Float3Write=true
   )
  ]
  rendererClasses [&NiagaraRibbonRendererProperties]
  renderers [{rendererClass=&NiagaraRibbonRendererProperties}]
 }
 NE_VC_DriftTrail_Ribbon_Parent001 {
  emitterUpdateScript [
   &EmitterState()
   &VC_SpawnRibbon_v1(Number of Ribbons=2, SpawnRate=120)
  ]
  particleSpawnScript [
   &VC_PaintedVariantsOverride() {enabled=false}
   &VC_InitializeRibbon_v1(
    Custom Offset=true
    Enable Skeletal Position=true
    Local Vector Offset=&AddVectorToPosition(Position=<-110 0 -31>, Vector=$User.offset)
    Offset Amount=1
    Ribbon Width=69
   )
   &AddVelocity(Velocity=<-1400 0 0>, Velocity Speed Scale=2) {module=AddVelocity001}
   &Color(Color=<rgba 0.2 0.9 0.03 1>) {enabled=false}
   &Color(Color=<rgba 0.4 0.8963 0.0296 1>) {module=Color001}
  ]
  particleUpdateScript [
   &CameraOffset(Camera Offset Amount=-10000) {enabled=false}
   &ParticleState()
   &SolveForcesAndVelocity(Clamp Velocity=false, Limit Acceleration=false)
   &ScaleColor(Scale Alpha=5, Scale RGB=&VectorFromFloat(Value=4), ScaleA=true, ScaleRGB=true)
   &ScaleRibbonWidth(Ribbon Width Scale=&FloatFromCurve(FloatCurve=DataInterface() {propertyValues {Curve {keys [{arriveTangent=1.7626, interpMode=RCIM_Cubic, leaveTangent=1.7626, tangentMode=RCTM_User, time=0, value=0.2}, <at 0.5 1 RCIM_Cubic>]}}}))
   &VC_UpdateRibbon_v1(Enable Skeletal Position=true)
   &DynamicMaterialParameters(
    Index 0 Param 1=0.66
    Index 0 Param 2=0.5
    Index 0 Param 3=0.01
    Index 0 Param 4=1
    Module.Index0Float0Write=true
    Module.Index0Float1Write=true
    Module.Index0Float2Write=true
    Module.Index0Float3Write=true
   )
  ]
  rendererClasses [&NiagaraRibbonRendererProperties]
  renderers [{rendererClass=&NiagaraRibbonRendererProperties}]
 }
 NE_VC_DriftTrail_WheelWrap_Parent {
  emitterUpdateScript [
   &EmitterState(
    Inactive Response=::NewEnumerator1
    Life Cycle Mode=::NewEnumerator1
    Loop Behavior=::NewEnumerator1
    Loop Duration Mode=::NewEnumerator1
    UseLoopDelay=false
   )
   &SpawnBurst_Instantaneous(Spawn Count=2, Use Spawn Probability=false, UseLoopCountLimit=false)
  ]
  particleSpawnScript [
   &VC_PaintedVariantsOverride() {enabled=false}
   &InitializeParticle(
    Color Mode=::NewEnumerator0
    Mesh Scale Mode=::NewEnumerator4
    Sprite Rotation Mode=::NewEnumerator2
    Sprite Size Mode=::NewEnumerator3
    UsePositionOffset=false
    Write Material Random=true
   )
   &PartitionParticles()
   &SkeletalMeshLocation(
    Bone / Socket Index=&Add_Integer(A=$Particles.PartitionParticles.Partition, B=4)
    Bone / Socket Sampling Mode=::NewEnumerator5
    Interpolate Mesh Sampling=false
    MaskSpawnGroup=false
    Mesh Sampling Type=::NewEnumerator6
    Skeletal Mesh=$System.SkeletalMesh
   )
   &Color(Color=<rgba 0.313 0.9 0.03 1>) {enabled=false}
   &Color(Color=<rgba 0.248 0.8 0.064 1>) {module=Color001}
  ]
  particleUpdateScript [
   &FN_DisableTireFX(
    Mask Index Attribute:NiagaraInt32=$Particles.PartitionParticles.Partition
    Mask Index FL=2
    Mask Index FR=3
    Mask Index RL=0
    Mask Index RR=1
   )
   &VC_WheelWrapScale(Wheel Index Attribute:NiagaraInt32=&Add_Integer(A=$Particles.PartitionParticles.Partition, B=2))
   &ParticleState(Kill Particles When Lifetime Has Elapsed=false)
   &UpdateMeshOrientation(Rotation Coordinate Space=::NewEnumerator2, Rotation Rate=1.5, Rotation Vector=<0 1 0>)
   &SkeletalMeshLocation(
    Bone / Socket Index=&Add_Integer(A=$Particles.PartitionParticles.Partition, B=4)
    Bone / Socket Sampling Mode=::NewEnumerator5
    MaskSpawnGroup=false
    Mesh Sampling Type=::NewEnumerator6
    Orientation Sampling=::NewEnumerator1
    Skeletal Mesh=$System.SkeletalMesh
   ) {module=SkeletalMeshLocation001}
   &ScaleColor(Scale RGB=&VectorFromFloat(Value=10), ScaleA=true, ScaleRGB=true)
   &SolveForcesAndVelocity(Clamp Velocity=false, Limit Acceleration=false)
  ]
  rendererClasses [&NiagaraMeshRendererProperties]
  renderers [{rendererClass=&NiagaraMeshRendererProperties, rendererIndex=0}]
 }
}
summary {
 systemName=NS_DriftTrail_FrogLTE2_MP
 emitters [
  {emitterName=MainTrail, rendererClasses [&NiagaraRibbonRendererProperties]}
  {emitterName=NE_VC_DriftTrail_Ribbon_Parent001, rendererClasses [&NiagaraRibbonRendererProperties]}
  {emitterName=NE_VC_DriftTrail_WheelWrap_Parent, rendererClasses [&NiagaraMeshRendererProperties]}
  {emitterName=Garage_Cone, rendererClasses [&NiagaraMeshRendererProperties, &NiagaraSpriteRendererProperties]}
  {emitterName=Garage_Cone001, rendererClasses [&NiagaraMeshRendererProperties, &NiagaraSpriteRendererProperties]}
 ]
 userVariables [
  User.bIsFrontendPreview {type=NiagaraBool}
  User.bIsPainted {type=NiagaraBool}
  User.ColorVariantID {type=NiagaraInt32}
  User.DTCarOffset {type=Vector3f, description="", defaultValue {struct=Vector3f, value=<0 0 0>}}
  User.offset {type=Vector3f, description="offset from wheels(set through cosmetic template)", defaultValue {struct=Vector3f, value=<0 0 0>}}
  User.PaintedColor {type=LinearColor, description="", defaultValue {struct=LinearColor, value=<rgba 1 1 1 1>}}
  User.PaintedScalar {type=NiagaraFloat, defaultValue=1, description=""}
  User.VehicleFlags {type=NiagaraInt32, defaultValue=806289408}
  User.Wheel_Radius_Array {
   type=NiagaraDataInterfaceArrayFloat
   defaultValue {
    struct=NiagaraExt_VariableValue_DataInterface
    value {
     dataInterface=&self:NiagaraDataInterfaceArrayFloat_1
     dataInterfaceClass=&NiagaraDataInterfaceArrayFloat
    }
   }
  }
  User.Wheel_Width_Array {
   type=NiagaraDataInterfaceArrayFloat
   defaultValue {
    struct=NiagaraExt_VariableValue_DataInterface
    value {
     dataInterface=&self:NiagaraDataInterfaceArrayFloat_0
     dataInterfaceClass=&NiagaraDataInterfaceArrayFloat
    }
   }
  }
  User.WheelTouchingArray {
   type=NiagaraDataInterfaceArrayFloat
   defaultValue {
    struct=NiagaraExt_VariableValue_DataInterface
    value {
     dataInterface=&self:NiagaraDataInterfaceArrayFloat_2
     dataInterfaceClass=&NiagaraDataInterfaceArrayFloat
    }
   }
  }
 ]
}
