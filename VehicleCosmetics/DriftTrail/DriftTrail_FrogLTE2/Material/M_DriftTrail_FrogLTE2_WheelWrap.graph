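# Unlit masked material for Niagara mesh particles, the wheel-wrap piece of the FrogLTE2 drift trail cosmetic.
# Emissive multiplies particle color by a contrast-adjusted, panning frog shimmer texture with a NoiseSpeed parameter, then runs through MF_FX_EmissiveAdjustments.
# An inverted mask sums four frog-mask samples, panned at two speeds in opposite directions over two offset UV sets rotated opposite ways by a quarter turn.
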Material /VehicleCosmetics/DriftTrail/DriftTrail_FrogLTE2/Material/M_DriftTrail_FrogLTE2_WheelWrap
graph {
 properties {
  bAutomaticallySetUsageInEditor=false
  bUsedWithNiagaraMeshParticles=true
  blendMode=BLEND_Masked
  shadingModel=MSM_Unlit
  usedShadingModels=Unlit
 }
 nodes [
  n1 = Time()
  n2 = ScalarParameter() {defaultValue=1, parameterName=NoiseSpeed}
  n3 = TextureSample(UVs=Panner(Coordinate=TextureCoordinate() {uTiling=2, vTiling=0.5}, Time=Multiply(A=@n1, B=@n2))) {texture=/DealershipSharedLibrary/FX/Textures/Diffuse/T_VCVFX_Frog_Shimmer_RGB}
  n4 = Constant() {r=0.25}
  MaterialFunctionCall(Emissive=Multiply(
    A=ParticleColor().RGB
    B=MaterialFunctionCall(In=@n3.G, Contrast=@n4) {materialFunction=&CheapContrast}.Result
   )) {materialFunction=&MF_FX_EmissiveAdjustments}
  n5 = TextureCoordinate() {uTiling=2}
  n6 = Constant() {r=0.25}
  n7 = MaterialFunctionCall(UVs=Add(A=@n5, B=Constant() {r=-0.1}), "Rotation Angle (0-1)"=@n6) {materialFunction=&CustomRotator}
  n8 = TextureSample(UVs=Panner(Coordinate=@n7.Rotated Values) {speedY=1}) {texture=/DealershipSharedLibrary/FX/Textures/TexturePackage/T_VCVFX_MaskFrog_RGB}
  n9 = TextureSample(UVs=Panner(Coordinate=@n7.Rotated Values) {speedY=0.5}) {texture=&T_VCVFX_MaskFrog_RGB}
  n10 = MaterialFunctionCall(UVs=Add(A=@n5, B=Constant() {r=0.1}), "Rotation Angle (0-1)"=OneMinus(None=@n6)) {materialFunction=&CustomRotator}
  n11 = TextureSample(UVs=Panner(Coordinate=@n10.Rotated Values) {speedY=-0.5}) {texture=&T_VCVFX_MaskFrog_RGB}
  n12 = TextureSample(UVs=Panner(Coordinate=@n10.Rotated Values) {speedY=-1}) {texture=&T_VCVFX_MaskFrog_RGB}
  OneMinus(None=Add(A=Add(A=@n8.R, B=@n9.B), B=Add(A=@n11.R, B=@n12.B)))
 ]
}
